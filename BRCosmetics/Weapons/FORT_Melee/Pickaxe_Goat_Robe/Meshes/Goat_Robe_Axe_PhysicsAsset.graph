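# Physics asset for the Goat Robe two-handed pickaxe, simulating two dangling ribbons on the weapon.
# A capsule covers the handle, and a small box master body joins to it with free swing and twist.
# Two ribbon chains, each made of two thin box bodies, hang from that master body.
# Their joints pair wide swing limits with a tighter twist limit.
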PhysicsAsset /BRCosmetics/Weapons/FORT_Melee/Pickaxe_Goat_Robe/Meshes/Goat_Robe_Axe_PhysicsAsset
bodies {
 Melee_TwoHanded_Handle [{shapeType=Capsule, length=97.867, radius=3.6815, center=<-0.7564 -1.4466 40.5863>}]
 dyn_ribbon_a_01 [{shapeType=Box, extentX=7, extentY=0.5, extentZ=10, center=<5 0 0>, rotation=<rot 90 0 0>}]
 dyn_ribbon_a_02 [{shapeType=Box, extentX=7, extentY=0.5, extentZ=10, center=<5 0 0>, rotation=<rot 90 0 0>}]
 dyn_ribbon_b_01 [{shapeType=Box, extentX=7, extentY=0.5, extentZ=10, center=<5 0 0>, rotation=<rot 90 0 0>}]
 dyn_ribbon_b_02 [{shapeType=Box, extentX=7, extentY=0.5, extentZ=10, center=<5 0 0>, rotation=<rot 90 0 0>}]
 dyn_ribbon_mstr [{shapeType=Box, extentX=1.01, extentY=1.01, extentZ=1.01}]
}
constraints [
 {
  bone1Name=dyn_ribbon_mstr
  bone2Name=Melee_TwoHanded_Handle
  swing1LimitDegrees=45
  swing1Motion=Free
  swing2LimitDegrees=45
  swing2Motion=Free
  twistLimitDegrees=45
  twistMotion=Free
 }
 {bone1Name=dyn_ribbon_a_01, bone2Name=dyn_ribbon_mstr, swing1LimitDegrees=75, swing2LimitDegrees=75, twistLimitDegrees=20}
 {bone1Name=dyn_ribbon_a_02, bone2Name=dyn_ribbon_a_01, swing1LimitDegrees=75, swing2LimitDegrees=75, twistLimitDegrees=20}
 {bone1Name=dyn_ribbon_b_01, bone2Name=dyn_ribbon_mstr, swing1LimitDegrees=75, swing2LimitDegrees=75, twistLimitDegrees=20}
 {bone1Name=dyn_ribbon_b_02, bone2Name=dyn_ribbon_b_01, swing1LimitDegrees=75, swing2LimitDegrees=75, twistLimitDegrees=20}
]
